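# Physics asset for the GumOutlaw backpack cosmetic mesh, giving its handle, zip, shells, grenade and ring secondary dangle motion.
# Tiny base anchor bodies lock their swing and twist to the main pack capsule, each carrying a dyn_ body on limited swing and twist constraints.
# Handle and shell pieces move within narrow limits, while the zip, grenade and ring form a looser hanging chain, zip to grenade to ring.
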PhysicsAsset /BRCosmetics/Accessories/FORT_Backpacks/Backpack_M_MED_GumOutlaw/Meshes/M_MED_GumOutlaw_Pack_Physics
bodies {
 base_handle [{shapeType=Capsule, length=0.505, radius=0.505, shapeName=base_handle_capsule, rotation=<rot 0 -90 0>}]
 base_l_back_shell [{shapeType=Capsule, length=0.505, radius=0.505, shapeName=base_l_back_shell_capsule, rotation=<rot 0 -90 0>}]
 base_l_top_shell [{shapeType=Capsule, length=0.505, radius=0.505, shapeName=base_l_top_shell_capsule, rotation=<rot 0 -90 0>}]
 base_r_back_shell [{shapeType=Capsule, length=0.505, radius=0.505, shapeName=base_r_back_shell_capsule, rotation=<rot 0 -90 0>}]
 base_zip [{shapeType=Capsule, length=0.505, radius=0.505, shapeName=base_zip_capsule, rotation=<rot 0 -90 0>}]
 dyn_grenade_00 [{shapeType=Sphere, radius=4, shapeName=dyn_grenade_00_sphere, center=<1.4587 0 4.463>}]
 dyn_handle_00 [{
   shapeType=Capsule
   length=6.5203
   radius=3.1977
   shapeName=dyn_handle_00_capsule
   center=<-0.6036 0.2328 3.1635>
   rotation=<rot 0 93.6543 -87.6748>
  }]
 dyn_l_back_shell_00 [{
   shapeType=Capsule
   length=6.2441
   radius=2.998
   shapeName=dyn_l_back_shell_00_capsule
   center=<0.5318 -0.4712 -0.2466>
   rotation=<rot 0 176.5468 -171.0538>
  }]
 dyn_l_top_shell_00 [{shapeType=Capsule, length=5.5, radius=0.5, shapeName=dyn_l_top_shell_00_capsule, center=<3.664 0 0>, rotation=<rot 0 0 89.9995>}]
 dyn_r_back_shell_00 [{
   shapeType=Capsule
   length=6.3973
   radius=3.3152
   shapeName=dyn_r_back_shell_00_capsule
   center=<1.5039 0.2094 -0.2658>
   rotation=<rot 0 1.8252 -167.9676>
  }]
 dyn_ring_00 [{shapeType=Box, extentX=0.5, extentY=4.5, extentZ=4.5, shapeName=dyn_ring_00_box, center=<0 -2.0461 0>}]
 dyn_zip_00 [{
   shapeType=Capsule
   length=0.505
   radius=0.505
   shapeName=dyn_zip_00_capsule
   center=<1.1406 -0.043 0.0108>
   rotation=<rot 0 85.7275 -85.3664>
  }]
 pack [{shapeType=Capsule, length=17.6417, radius=8.2682, shapeName=pack_capsule, center=<0.3959 -8.4039 -6.6924>, rotation=<rot 0 89.7607 -50.8252>}]
}
constraints [
 {bone1Name=dyn_handle_00, bone2Name=base_handle, swing1LimitDegrees=3, swing2LimitDegrees=3, twistLimitDegrees=25}
 {bone1Name=dyn_zip_00, bone2Name=base_zip, swing1LimitDegrees=45, swing2LimitDegrees=20, twistLimitDegrees=0}
 {bone1Name=dyn_ring_00, bone2Name=dyn_grenade_00, swing1LimitDegrees=45, swing2LimitDegrees=45, twistLimitDegrees=15}
 {bone1Name=dyn_l_top_shell_00, bone2Name=base_l_top_shell, swing1LimitDegrees=5, swing2LimitDegrees=12, twistLimitDegrees=5}
 {
  bone1Name=base_r_back_shell
  bone2Name=pack
  swing1LimitDegrees=30
  swing1Motion=Locked
  swing2LimitDegrees=30
  swing2Motion=Locked
  twistLimitDegrees=30
  twistMotion=Locked
 }
 {
  bone1Name=base_l_back_shell
  bone2Name=pack
  swing1LimitDegrees=45
  swing1Motion=Locked
  swing2LimitDegrees=45
  swing2Motion=Locked
  twistLimitDegrees=45
  twistMotion=Locked
 }
 {
  bone1Name=base_l_top_shell
  bone2Name=pack
  swing1LimitDegrees=45
  swing1Motion=Locked
  swing2LimitDegrees=45
  swing2Motion=Locked
  twistLimitDegrees=45
  twistMotion=Locked
 }
 {
  bone1Name=base_handle
  bone2Name=pack
  swing1LimitDegrees=45
  swing1Motion=Locked
  swing2LimitDegrees=45
  swing2Motion=Locked
  twistLimitDegrees=45
  twistMotion=Locked
 }
 {
  bone1Name=base_zip
  bone2Name=pack
  swing1LimitDegrees=45
  swing1Motion=Locked
  swing2LimitDegrees=45
  swing2Motion=Locked
  twistLimitDegrees=45
  twistMotion=Locked
 }
 {bone1Name=dyn_l_back_shell_00, bone2Name=base_l_back_shell, swing1LimitDegrees=0, swing2LimitDegrees=5, twistLimitDegrees=0}
 {bone1Name=dyn_r_back_shell_00, bone2Name=base_r_back_shell, swing1LimitDegrees=0, swing2LimitDegrees=5, twistLimitDegrees=0}
 {bone1Name=dyn_grenade_00, bone2Name=dyn_zip_00, swing1LimitDegrees=60, swing2LimitDegrees=60, twistLimitDegrees=0}
]
